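# Niagara system that dresses the StuffRasp Maze character cosmetic in pink-red fire attached to skeleton bones.
# Mesh emitters form body flames, limb and pelvis swirls, eye fire and bone wisps, while a sprite emitter adds body spheres.
# Partition-indexed arrays set bones, offsets and sizes, with smoothed positions feeding some flames.
# Music amplitude from an NPC drives opacity, and user booleans hide foot and head pieces or resize swirls for front-end preview.
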NiagaraSystem /BRCosmetics/Characters/Player/Male/Medium/Bodies/M_MED_StuffRasp_Maze/FX/NS_Character_StuffRasp_Veil_BoneFire
dependencies {
 usedDataInterfaces [
  &NiagaraDataInterfaceArrayFloat3
  &NiagaraDataInterfaceArrayInt32
  &NiagaraDataInterfaceCurve
  &NiagaraDataInterfaceArrayFloat
  &NiagaraDataInterfaceSkeletalMesh
 ]
 usedDynamicInputs []
 usedModules [
  &self:SystemSpawnScript.SystemScriptSource.SystemScriptGraph.NiagaraNodeAssignment_1.SetVariables_#1
  &SystemState
 ]
}
emitters {
 FireBodyStatic {
  emitterUpdateScript [
   &EmitterState(
    Inactive Response=::NewEnumerator1
    Life Cycle Mode=::NewEnumerator1
    Loop Behavior=::NewEnumerator1
    Loop Delay=0.03
    Loop Duration Mode=::NewEnumerator1
    UseLoopDelay=false
   )
   &SpawnBurst_Instantaneous(Spawn Count=21, Use Spawn Probability=false, UseLoopCountLimit=false)
   &SpawnBurst_Instantaneous(Spawn Count=21, Spawn Group=1, Use Spawn Probability=false, UseLoopCountLimit=false) {module=SpawnBurst_Instantaneous001}
  ]
  particleSpawnScript [
   /SetVariables(Particles.VisibilityTag=&SpawnGroup()) {isSetParameters=true}
   &PartitionParticles(Number Of Partitions=21)
   &InitializeParticle(
    Color Maximum=<rgba 1 0.4792 0.5401 0.1>
    Color Minimum=<rgba 1 0.2135 0.3056 0.1>
    Color Mode=::NewEnumerator2
    Mesh Scale=&Multiply_Vector(
     A=&SelectVectorFromArray(
      Direct Array Index=$Particles.PartitionParticles.Partition
      Vector Selection Array=DataInterface() {propertyValues {FloatData [
         <0.8 0.7 8>
         <0.8 0.8 5>
         <0.5 0.8 12>
         <0.5 1 -20>
         <0.7 0.8 30>
         <0.6 1.2 -60>
         <0.6 1 30>
         <0.6 1 -40>
         <0.6 0.7 30>
         <0.6 0.75 -30>
         <0.7 1.5 -80>
         <0.7 1 60>
         <0.7 1 -50>
         <0.7 0.8 60>
         <0.6 1.5 -75>
         <0.6 1.5 80>
         <0.6 1.5 -50>
         <0.5 1 70>
         <0.5 0.8 50>
         <0.6 0.8 1>
         <0.6 0.8 1>
        ]}}
     )
     B=<0.95 0.95 1>
    )
    Mesh Scale Mode=::NewEnumerator4
    Sprite Size Mode=::NewEnumerator3
    Uniform Sprite Size=20
    UsePositionOffset=false
    Write Material Random=true
   )
   &InitialMeshOrientation(
    Mesh Orientation Mode=::NewEnumerator2
    Rotation=&SelectVectorFromArray(
     Direct Array Index=$Particles.PartitionParticles.Partition
     Vector Selection Array=DataInterface() {propertyValues {FloatData [
        <0 0 0.25>
        <0 0 0.25>
        <0 0 0.25>
        <0 0 0.25>
        <0 0 0.25>
        <0 0 0.25>
        <0 0 0.25>
        <0 0 0.25>
        <0 0 0.25>
        <0 0 0.25>
        <0 0 0.25>
        <0 0 0.25>
        <0 0 0.25>
        <0 0 0.25>
        <0 0 0.25>
        <0 0 0.25>
        <0 0 0.25>
        <0 0 0.25>
        <0 0 0.25>
        <0 0 0.25>
       ]}}
    )
    Use Rotation Vector=true
   )
   &CameraOffset(Camera Offset Amount=-4)
  ]
  particleUpdateScript [
   &ParticleState(Kill Particles When Lifetime Has Elapsed=false, Let Infinitely Lived Particles Die When Emitter Deactivates=true)
   &ScaleMeshSize(Scale Factor=&MakeCustomVectorFromBool(
     Bool=$User.bisFootHidden
     True Vector=&Multiply_VectorByFloat(
      Vector=&VectorFromFloat(Value=&SelectFloatFromArray(
        Direct Array Index=$Particles.PartitionParticles.Partition
        Float Selection Array=DataInterface() {propertyValues {FloatData [1,1,1,1,0,0,0,0,0,0,1,1,1,1,0,0,0,0,0]}}
       ))
      Float=&MakeFloatFromVector(
       Vector=&VectorFromFloat(Value=&MaskFloatBySpawnGroup(Masked Float=0.7, Passthrough Float=1.05))
       Channel=::NewEnumerator0
      )
     )
     False Vector=&VectorFromFloat(Value=&MaskFloatBySpawnGroup(Masked Float=0.7, Passthrough Float=1.05))
    ))
   &SkeletalMeshLocation(
    Bone / Socket Index=&SelectIntFromArray(
     Direct Array Index=$Particles.PartitionParticles.Partition
     Int Selection Array=DataInterface() {propertyValues {IntData [20,21,1,2,1,2,5,6,5,6,9,10,9,10,9,10,13,14,14,20,21]}}
    )
    Bone Sampling Mode=::NewEnumerator4
    MaskSpawnGroup=false
    Sampled Position Offset=&SelectVectorFromArray(
     Direct Array Index=$Particles.PartitionParticles.Partition
     Vector Selection Array=DataInterface() {propertyValues {FloatData [
        <15 0 0>
        <-12 0 -2>
        <5 0 3>
        <-5 0 0>
        <17 0 2>
        <-15 0 0>
        <6 0 2>
        <-6 0 -2>
        <18 0 0>
        <-18 0 0>
        <-5 0 3>
        <0 0 2>
        <-20 0 -0.5>
        <15 0 2>
        <-30 0 2>
        <23 0 -1>
        <-20 0 0>
        <13 0 -0.5>
        <28 0 0>
        <5 0 5>
        <-5 0 -4>
       ]}}
    )
    Skeletal Mesh=$System.Skeletal Mesh
   )
   &DynamicMaterialParameters(
    Index 0 Param 1=&RandomRangeFloat(Minimum=0.4, Maximum=0.8)
    Index 0 Param 2=&RandomRangeFloat(Minimum=0.4, Maximum=1)
    Index 0 Param 3=1
    Index 0 Param 4=&RandomRangeFloat(Minimum=0.8, Maximum=1)
    Index 2 Param 1=&RandomRangeFloat(Minimum=0.8, Maximum=1)
    Index 2 Param 2=&Multiply_Float(
     A=&SelectFloatFromArray(
      Direct Array Index=$Particles.PartitionParticles.Partition
      Float Selection Array=DataInterface() {propertyValues {FloatData [0.7,0.8,0.5,0.6,0.2,0.2,0.6,0.125,0.75,0.6,1,1,1,0.5,1,0.2,0.9,1,0.8,1.5,1.5]}}
     )
     B=&FloatFromCurve(FloatCurve=DataInterface() {propertyValues {Curve {keys [<at 0 1.2>,<at 0.9 1.2>]}}})
    )
    Module.Index0Float0Write=true
    Module.Index0Float1Write=true
    Module.Index0Float2Write=true
    Module.Index0Float3Write=true
    Module.Index2Float0Write=true
    Module.Index2Float1Write=true
    Module.Index2Float2Write=true
    Module.Index2Float3Write=true
    Write Parameter Index 2=::NewEnumerator1
   )
   &InterpolateOverTime(Rate Of Change=5, Target Vector Value:NiagaraPosition=$Particles.Position, Variable Type=::NewEnumerator2)
   /SetVariables(Particles.Interpolate Dynamic=&MakeFloat4FromVectorAndFloat("Vector (RGB)"=$Particles.InterpolateOverTime.InterpolateOverTime_V3_.Moving Average, "Float (Alpha)"=0)) {isSetParameters=true}
   &ScaleColor(
    Scale Alpha=&Lerp_Float(
     A=&RandomRangeFloat(Minimum=0.135, Maximum=0.3)
     B=&RandomRangeFloat(Minimum=0.45, Maximum=1)
     Alpha=&RemapRange(Input Value To Remap=$NPC.NPC_Audio_Reactive_Parameters.Amplitude_Music, Input Min=0.4, Input Max=0.925, Output Min=0, Output Max=1)
    )
    Scale RGB=&MakeVectorFromLinearColorRGB(Color=&MultiplyLinearColor(
      Linear Color=&MaskLinearColorBySpawnGroup(Masked Linear Color=<rgba 2 0 0.2341 1>, Passthrough Linear Color=<rgba 0 0 0 1>)
      Scale Factor=&MakeVector4_FromFloat(Float=1)
     ))
    ScaleA=true
    ScaleRGB=true
   )
   &ScaleColor(
    Scale Alpha=&SelectFloatFromArray(
     Direct Array Index=$Particles.PartitionParticles.Partition
     Float Selection Array=DataInterface() {propertyValues {FloatData [1,1,1,1,0,0,0,0,0,0,0.4,0.4,0.4,0.4,0.4,0.4,1,1,0]}}
    )
    ScaleA=true
    ScaleRGB=true
   ) {module=ScaleColor001}
  ]
  rendererClasses [&NiagaraMeshRendererProperties]
  renderers [
   {rendererClass=&NiagaraMeshRendererProperties, rendererIndex=0}
   {rendererClass=&NiagaraMeshRendererProperties, rendererIndex=1}
   {rendererClass=&NiagaraMeshRendererProperties, rendererIndex=2}
  ]
 }
 FireSwirls_ArmThigh {
  emitterSpawnScript [/SetVariables(Emitter.TimeMult:NiagaraFloat=&RandomRangeFloat(Minimum=1.5, Maximum=2)) {isSetParameters=true, module=SetVariables_#2}]
  emitterUpdateScript [
   &EmitterState(Inactive Response=::NewEnumerator1, Life Cycle Mode=::NewEnumerator1, Loop Duration=1, UseLoopDelay=false)
   &SpawnRate(SpawnRate=3, Use Spawn Probability=false)
  ]
  particleSpawnScript [
   &PartitionParticles(Number Of Partitions=8)
   &InitializeParticle(
    Color Maximum=<rgba 5 0.5469 1.068 0.9>
    Color Minimum=<rgba 2 0.0938 0.3168 0.7>
    Color Mode=::NewEnumerator2
    Lifetime Max=1.7
    Lifetime Min=1.3
    Lifetime Mode=::NewEnumerator1
    Mesh Scale=&Multiply_Vector(
     A=&SelectVectorFromArray(
      Direct Array Index=$Particles.PartitionParticles.Partition
      Vector Selection Array=DataInterface() {propertyValues {FloatData [<1 0.8 -0.8>, <1.6 1.1 -1.1>, <1 0.8 0.8>, <1.6 1.1 -1.1>, <0.8 0.8 -0.8>, <1.1 1 -1>, <0.8 0.8 0.8>, <1.1 1 1>]}}
     )
     B=&RandomRangeVector(Minimum=<0.8 1 1>, Maximum=<1 1 1>)
    )
    Mesh Scale Mode=::NewEnumerator4
    UsePositionOffset=false
    Write Material Random=true
   )
   &InitialMeshOrientation(
    Mesh Orientation Mode=::NewEnumerator2
    Rotation=&SelectVectorFromArray(
     Direct Array Index=$Particles.PartitionParticles.Partition
     Vector Selection Array=DataInterface() {propertyValues {FloatData [<0 0 0.5>, <0 0 0.5>, <0 0 0>, <0 0 0>, <0 0 0>, <0 0 0>, <0 0 0.5>, <0 0 0.5>]}}
    )
    Use Rotation Vector=true
   )
   &InitialMeshOrientation(
    Mesh Orientation Mode=::NewEnumerator2
    Rotation=&RandomRangeVector(Minimum=<0 0 0>, Maximum=<1 0 0>)
    Use Rotation Vector=true
   ) {module=InitialMeshOrientation001}
   &InitialMeshOrientation(
    Mesh Orientation Mode=::NewEnumerator2
    Rotation=&SelectVectorFromArray(
     Direct Array Index=$Particles.PartitionParticles.Partition
     Vector Selection Array=DataInterface() {propertyValues {FloatData [<0 0 0>, <0 0.01 0>, <0 0 0>, <0 0.01 0>, <0 0 0>, <0 0 0>, <0 0 0>, <0 0 0>]}}
    )
    Use Rotation Vector=true
   ) {module=InitialMeshOrientation002}
   &CameraOffset(Camera Offset Amount=15)
  ]
  particleUpdateScript [
   &ParticleState()
   &ScaleMeshSize(Scale Factor=&MakeCustomVectorFromBool(Bool=$User.bIsFrontEndPreview, True Vector=<1 0.8 0.8>, False Vector=<1 1 1>))
   &SkeletalMeshLocation(
    Bone / Socket Index=&SelectIntFromArray(
     Direct Array Index=$Particles.PartitionParticles.Partition
     Int Selection Array=DataInterface() {propertyValues {IntData [14,10,13,9,6,2,5,1]}}
    )
    Bone Sampling Mode=::NewEnumerator4
    MaskSpawnGroup=false
    Sampled Position Offset=&SelectVectorFromArray(
     Direct Array Index=$Particles.PartitionParticles.Partition
     Vector Selection Array=DataInterface() {propertyValues {FloatData [<30 2 -2>, <40 0 -2>, <-30 -2 2>, <-40 0 1>, <-20 0 0>, <-25 0 0>, <20 0 0>, <25 0 0>]}}
    )
    Skeletal Mesh=$System.Skeletal Mesh
   )
   /SetVariables(Particles.Width:NiagaraFloat=&RandomRangeFloat(Minimum=0.4, Maximum=0.6)) {isSetParameters=true, module=SetVariables_#3}
   &DynamicMaterialParameters(
    Index 0 Param 1=&Multiply_Float(A=15, B=$Particles.Width)
    Index 0 Param 2=&Multiply_Float(A=15, B=$Particles.Width)
    Index 0 Param 3=&RandomRangeFloat(Minimum=0, Maximum=0.8)
    Index 1 Param 1=&RandomRangeFloat(Minimum=1, Maximum=2)
    Index 2 Param 1=&RandomRangeFloat(Minimum=0.7, Maximum=1.3)
    Module.Index0Param0Write=true
    Module.Index0Param1Write=true
    Module.Index0Param2Write=true
    Module.Index0Param3Write=true
    Module.Index1Param0Write=true
    Module.Index1Param1Write=true
    Module.Index1Param2Write=true
    Module.Index1Param3Write=true
    Module.Index2Param0Write=true
    Module.Index2Param1Write=true
    Module.Index2Param2Write=true
    Module.Index2Param3Write=true
    Write Parameter Index 1=true
    Write Parameter Index 2=true
   )
   &SolveForcesAndVelocity(Clamp Velocity=false, Limit Acceleration=false)
   &Timeline(
    Play=&SetBoolByFloatComparison(A=$NPC.NPC_Audio_Reactive_Parameters.Amplitude_Music, B=0.675)
    Rewind Rate=0.55
   )
   &ScaleColor(
    Scale Alpha=&FloatFromCurve(
     FloatCurve=DataInterface() {propertyValues {Curve {
        defaultValue=1
        keys [
         {interpMode=RCIM_Cubic, tangentMode=RCTM_User, time=0, value=0}
         {interpMode=RCIM_Cubic, tangentMode=RCTM_User, time=0.5, value=1}
         {tangentMode=RCTM_User, time=1, value=0}
        ]
       }}}
     Scale Curve=&Lerp_Float(
      A=&RandomRangeFloat(Minimum=0.24, Maximum=0.3)
      B=&RandomRangeFloat(Minimum=0.8, Maximum=1)
      Alpha=&RemapRange(Input Value To Remap=$NPC.NPC_Audio_Reactive_Parameters.Amplitude_Music, Input Min=0.4, Input Max=0.925, Output Min=0, Output Max=1)
     )
    )
    ScaleA=true
    ScaleRGB=true
   )
   &ScaleMeshSize(Scale Factor=&VectorFromFloat(Value=&SelectFloatFromArray(
      Direct Array Index=$Particles.PartitionParticles.Partition
      Float Selection Array=DataInterface() {propertyValues {FloatData [0,1,0,1,0,1,0,1]}}
     ))) {module=ScaleMeshSize001}
   &ScaleColor(
    Scale Alpha=&SelectFloatFromArray(
     Direct Array Index=$Particles.PartitionParticles.Partition
     Float Selection Array=DataInterface() {propertyValues {FloatData [0.3,0.3,0.3,0.3,0,0,0,0]}}
    )
    ScaleA=true
    ScaleRGB=true
   ) {module=ScaleColor001}
  ]
  rendererClasses [&NiagaraMeshRendererProperties]
  renderers [{rendererClass=&NiagaraMeshRendererProperties, rendererIndex=0}]
 }
 FireSwirls_Calf {
  emitterSpawnScript [/SetVariables(Emitter.TimeMult:NiagaraFloat=&RandomRangeFloat(Minimum=1.5, Maximum=2)) {isSetParameters=true, module=SetVariables_#2}]
  emitterUpdateScript [
   &EmitterState(Inactive Response=::NewEnumerator1, Life Cycle Mode=::NewEnumerator1, Loop Duration=1, UseLoopDelay=false)
   &SpawnRate(SpawnRate=24, Use Spawn Probability=false)
  ]
  particleSpawnScript [
   &PartitionParticles(Number Of Partitions=8)
   &InitializeParticle(
    Color Maximum=<rgba 5 0.5469 1.068 0.9>
    Color Minimum=<rgba 2 0.0938 0.3168 0.7>
    Color Mode=::NewEnumerator2
    Lifetime Max=1.3
    Lifetime Min=0.9
    Lifetime Mode=::NewEnumerator1
    Mesh Scale=&Multiply_Vector(
     A=&SelectVectorFromArray(
      Direct Array Index=$Particles.PartitionParticles.Partition
      Vector Selection Array=DataInterface() {propertyValues {FloatData [<1.45 1 -1.4>, <1.6 1.1 -1.1>, <1.45 1 1.4>, <1.6 1.1 -1.1>, <0.8 0.8 -0.8>, <1.1 1 -1>, <0.8 0.8 0.8>, <1.1 1 1>]}}
     )
     B=&RandomRangeVector(Minimum=<0.8 1 1>, Maximum=<1 1 1>)
    )
    Mesh Scale Mode=::NewEnumerator4
    UsePositionOffset=false
    Write Material Random=true
   )
   &InitialMeshOrientation(
    Mesh Orientation Mode=::NewEnumerator2
    Rotation=&SelectVectorFromArray(
     Direct Array Index=$Particles.PartitionParticles.Partition
     Vector Selection Array=DataInterface() {propertyValues {FloatData [<0 0 0.5>, <0 0 0.5>, <0 0 0>, <0 0 0>, <0 0 0>, <0 0 0>, <0 0 0.5>, <0 0 0.5>]}}
    )
    Use Rotation Vector=true
   )
   &InitialMeshOrientation(
    Mesh Orientation Mode=::NewEnumerator2
    Rotation=&RandomRangeVector(Minimum=<0 0 0>, Maximum=<1 0 0>)
    Use Rotation Vector=true
   ) {module=InitialMeshOrientation001}
   &InitialMeshOrientation(
    Mesh Orientation Mode=::NewEnumerator2
    Rotation=&SelectVectorFromArray(
     Direct Array Index=$Particles.PartitionParticles.Partition
     Vector Selection Array=DataInterface() {propertyValues {FloatData [<0 0 0>, <0 0.01 0>, <0 0 0>, <0 0.01 0>, <0 0 0>, <0 0 0>, <0 0 0>, <0 0 0>]}}
    )
    Use Rotation Vector=true
   ) {module=InitialMeshOrientation002}
   &CameraOffset(Camera Offset Amount=15)
  ]
  particleUpdateScript [
   &ParticleState()
   &ScaleMeshSize(Scale Factor=&MakeCustomVectorFromBool(Bool=$User.bIsFrontEndPreview, True Vector=<1 0.8 0.8>, False Vector=<1 1 1>))
   &SkeletalMeshLocation(
    Bone / Socket Index=&SelectIntFromArray(
     Direct Array Index=$Particles.PartitionParticles.Partition
     Int Selection Array=DataInterface() {propertyValues {IntData [14,10,13,9,6,2,5,1]}}
    )
    Bone Sampling Mode=::NewEnumerator4
    MaskSpawnGroup=false
    Sampled Position Offset=&SelectVectorFromArray(
     Direct Array Index=$Particles.PartitionParticles.Partition
     Vector Selection Array=DataInterface() {propertyValues {FloatData [<30 2 -2>, <40 0 -2>, <-30 -2 2>, <-40 0 1>, <-20 0 0>, <-25 0 0>, <20 0 0>, <25 0 0>]}}
    )
    Skeletal Mesh=$System.Skeletal Mesh
   )
   /SetVariables(Particles.Width:NiagaraFloat=&RandomRangeFloat(Minimum=0.4, Maximum=0.6)) {isSetParameters=true, module=SetVariables_#3}
   &DynamicMaterialParameters(
    Index 0 Param 1=&Multiply_Float(A=15, B=$Particles.Width)
    Index 0 Param 2=&Multiply_Float(A=15, B=$Particles.Width)
    Index 0 Param 3=&RandomRangeFloat(Minimum=0, Maximum=0.8)
    Index 1 Param 1=&RandomRangeFloat(Minimum=1, Maximum=2)
    Index 2 Param 1=&RandomRangeFloat(Minimum=0.7, Maximum=1.3)
    Module.Index0Param0Write=true
    Module.Index0Param1Write=true
    Module.Index0Param2Write=true
    Module.Index0Param3Write=true
    Module.Index1Param0Write=true
    Module.Index1Param1Write=true
    Module.Index1Param2Write=true
    Module.Index1Param3Write=true
    Module.Index2Param0Write=true
    Module.Index2Param1Write=true
    Module.Index2Param2Write=true
    Module.Index2Param3Write=true
    Write Parameter Index 1=true
    Write Parameter Index 2=true
   )
   &SolveForcesAndVelocity(Clamp Velocity=false, Limit Acceleration=false)
   &Timeline(
    Play=&SetBoolByFloatComparison(A=$NPC.NPC_Audio_Reactive_Parameters.Amplitude_Music, B=0.675)
    Rewind Rate=0.55
   )
   &ScaleColor(
    Scale Alpha=&FloatFromCurve(
     FloatCurve=DataInterface() {propertyValues {Curve {
        defaultValue=0
        keys [
         {interpMode=RCIM_Cubic, tangentMode=RCTM_User, time=0, value=0}
         {interpMode=RCIM_Cubic, tangentMode=RCTM_User, time=0.5, value=1}
         {tangentMode=RCTM_User, time=1, value=0}
        ]
       }}}
     Scale Curve=&Lerp_Float(
      A=&RandomRangeFloat(Minimum=0.12, Maximum=0.24)
      B=&RandomRangeFloat(Minimum=0.4, Maximum=0.8)
      Alpha=&RemapRange(Input Value To Remap=$NPC.NPC_Audio_Reactive_Parameters.Amplitude_Music, Input Min=0.4, Input Max=0.925, Output Min=0, Output Max=1)
     )
    )
    ScaleA=true
    ScaleRGB=true
   )
   &ScaleMeshSize(Scale Factor=&VectorFromFloat(Value=&MakeCustomFloatFromBool(
      Bool=$User.bisFootHidden
      True Float=0
      False Float=&SelectFloatFromArray(
       Direct Array Index=$Particles.PartitionParticles.Partition
       Float Selection Array=DataInterface() {propertyValues {FloatData [1,0,1,0,0,0,0,0]}}
      )
     ))) {module=ScaleMeshSize001}
  ]
  rendererClasses [&NiagaraMeshRendererProperties]
  renderers [{rendererClass=&NiagaraMeshRendererProperties, rendererIndex=0}]
 }
 FireSwirls_Pelvis {
  emitterUpdateScript [
   &EmitterState(
    Inactive Response=::NewEnumerator1
    Life Cycle Mode=::NewEnumerator1
    Loop Behavior=::NewEnumerator1
    Loop Duration Mode=::NewEnumerator1
    UseLoopDelay=false
   )
   &SpawnBurst_Instantaneous(Use Spawn Probability=false, UseLoopCountLimit=false)
  ]
  particleSpawnScript [
   &InitializeParticle(
    Color=&RandomRangeLinearColor(Minimum=<rgba 2 0.1042 0.326 1>, Maximum=<rgba 4 0.2292 0.6705 1>)
    Mesh Scale=<-1 -1 0.6>
    Mesh Scale Mode=::NewEnumerator4
    UsePositionOffset=false
    Write Material Random=true
   )
   &InitialMeshOrientation(Mesh Orientation Mode=::NewEnumerator2, Rotation=<0 0.25 0>, Use Rotation Vector=true)
   &CameraOffset(Camera Offset Amount=15)
  ]
  particleUpdateScript [
   &ParticleState(Kill Particles When Lifetime Has Elapsed=false, Let Infinitely Lived Particles Die When Emitter Deactivates=true)
   &SkeletalMeshLocation(
    Bone / Socket Index=19
    Bone Sampling Mode=::NewEnumerator4
    MaskSpawnGroup=false
    Sampled Position Offset=<-12.5 0 -1.25>
    Skeletal Mesh=$System.Skeletal Mesh
   )
   &SolveForcesAndVelocity(Clamp Velocity=false, Limit Acceleration=false)
   &DynamicMaterialParameters(
    Index 0 Param 1=10
    Index 0 Param 2=8
    Index 0 Param 3=1
    Index 0 Param 4=1
    Module.Index0Float0Write=true
    Module.Index0Float1Write=true
    Module.Index0Float2Write=true
    Module.Index0Float3Write=true
   )
   &Timeline(
    Play=&SetBoolByFloatComparison(A=$NPC.NPC_Audio_Reactive_Parameters.Amplitude_Music, B=0.675)
    Rewind Rate=0.55
   )
   &ScaleColor(
    Scale Alpha=&Lerp_Float(
     A=0.21
     B=0.7
     Alpha=&RemapRange(Input Value To Remap=$NPC.NPC_Audio_Reactive_Parameters.Amplitude_Music, Input Min=0.4, Input Max=0.925, Output Min=0, Output Max=1)
    )
    ScaleA=true
    ScaleRGB=true
   )
  ]
  rendererClasses [&NiagaraMeshRendererProperties]
  renderers [{rendererClass=&NiagaraMeshRendererProperties, rendererIndex=0}]
 }
 Fire_Eyes {
  emitterSpawnScript [/SetVariables(Emitter.TimeMult:NiagaraFloat=&RandomRangeFloat(Minimum=-1.5, Maximum=-2)) {isSetParameters=true}]
  emitterUpdateScript [
   &EmitterState(Life Cycle Mode=::NewEnumerator1, Loop Behavior=::NewEnumerator1, Loop Duration Mode=::NewEnumerator1, UseLoopDelay=false)
   &SpawnBurst_Instantaneous(Use Spawn Probability=false, UseLoopCountLimit=false)
   &SpawnBurst_Instantaneous(Spawn Group=1, Use Spawn Probability=false, UseLoopCountLimit=false) {module=SpawnBurst_Instantaneous001}
  ]
  particleSpawnScript [
   /SetVariables(Particles.VisibilityTag=&SpawnGroup()) {isSetParameters=true}
   &InitializeParticle(
    Color=&MaskLinearColorBySpawnGroup(Masked Linear Color=<rgba 3 0.7344 0.9995 1>, Passthrough Linear Color=<rgba 1 0.0545 0.1651 1>)
    Mesh Scale Mode=::NewEnumerator3
    UsePositionOffset=false
    Write Material Random=true
   )
   &InitialMeshOrientation(Mesh Orientation Mode=::NewEnumerator1, Use Rotation Vector=true)
  ]
  particleUpdateScript [
   &ParticleState(Kill Particles When Lifetime Has Elapsed=false, Let Infinitely Lived Particles Die When Emitter Deactivates=true)
   &ScaleMeshSize(Scale Factor=&VectorFromFloat(Value=&MakeCustomFloatFromBool(Bool=$User.bisHeadHidden, True Float=0, False Float=1)))
   &DynamicMaterialParameters(
    Index 0 Param 1=&MaskFloat(Mask Source=0, Mask Parameter=?, Mask=0, Masked Float=5, Passthrough Float=7)
    Index 0 Param 2=1
    Index 0 Param 3=&RandomRangeFloat(Minimum=0, Maximum=0.8)
    Index 1 Param 1=&RandomRangeFloat(Minimum=1, Maximum=2)
    Index 2 Param 1=&RandomRangeFloat(Minimum=0.7, Maximum=1.3)
    Module.Index0Param0Write=true
    Module.Index0Param1Write=true
    Module.Index0Param2Write=true
    Module.Index0Param3Write=true
    Module.Index1Param0Write=true
    Module.Index1Param1Write=true
    Module.Index1Param2Write=true
    Module.Index1Param3Write=true
    Module.Index2Param0Write=true
    Module.Index2Param1Write=true
    Module.Index2Param2Write=true
    Module.Index2Param3Write=true
    Write Parameter Index 1=true
    Write Parameter Index 2=true
   )
   &SkeletalMeshLocation(
    Bone Sampling Mode=::NewEnumerator4
    MaskSpawnGroup=false
    Sampled Position Offset Coordinate Space:ENiagaraSkelMeshTransforms=::NewEnumerator2
    Skeletal Mesh=$System.Skeletal Mesh
   )
   &Timeline(Play=&SetBoolByFloatComparison(A=$NPC.NPC_Audio_Reactive_Parameters.Amplitude_Music, B=0.675), Rewind Rate=0.6)
   &ScaleColor(
    Scale Alpha=&Lerp_Float(
     A=0.4
     B=&MaskFloat(Mask Source=0, Mask Parameter=?, Mask=0, Masked Float=0.4, Passthrough Float=3)
     Alpha=$Particles.Timeline.NormalizedTime
    )
    ScaleA=true
    ScaleRGB=true
   )
  ]
  rendererClasses [&NiagaraMeshRendererProperties]
  renderers [
   {rendererClass=&NiagaraMeshRendererProperties, rendererIndex=0}
   {rendererClass=&NiagaraMeshRendererProperties, rendererIndex=1}
   {rendererClass=&NiagaraMeshRendererProperties, rendererIndex=2}
  ]
 }
 Minimal {
  emitterSpawnScript [/SetVariables(Emitter.UseSplineThicken=true) {isSetParameters=true}]
  emitterUpdateScript [
   &EmitterState(Life Cycle Mode=::NewEnumerator1, Loop Behavior=::NewEnumerator1, Loop Duration Mode=::NewEnumerator1, UseLoopDelay=false)
   &SpawnRate(SpawnRate=4, Use Spawn Probability=false)
   &SpawnRate(SpawnGroup=1, SpawnRate=4, Use Spawn Probability=false) {module=SpawnRate001}
  ]
  particleSpawnScript [
   /SetVariables(Particles.VisibilityTag=&SpawnGroup()) {isSetParameters=true}
   &InitializeParticle(
    Color=<rgba 3 0.2812 0.5994 1>
    Lifetime Max=2.2
    Lifetime Min=0.8
    Lifetime Mode=::NewEnumerator1
    Mesh Scale=&MaskVectorBySpawnGroup(Masked Vector=<1 1 1.325>, Passthrough Vector=<1 1 1.325>)
    Mesh Scale Mode=::NewEnumerator4
    UsePositionOffset=false
    Write Material Random=true
   )
   /SetVariables(Particles.SplineSize:Vector4f=&MakeVector4(X=&MaskFloatBySpawnGroup(Masked Float=16, Passthrough Float=18), Y=10, Z=0, W=0)) {isSetParameters=true}
   /SetVariables(Particles.BlackSplineSize:Vector4f=&MakeVector4(X=&MaskFloatBySpawnGroup(Masked Float=18, Passthrough Float=20), Y=13, Z=0, W=0)) {isSetParameters=true}
   &CameraOffset(Camera Offset Amount=-10)
  ]
  particleUpdateScript [
   &ParticleState()
   &SkeletalMeshLocation(
    Bone / Socket Index=&MaskIntBySpawnGroup(Masked Int=20, Passthrough Int=21, Spawn Group Mask=0)
    Bone Sampling Mode=::NewEnumerator4
    MaskSpawnGroup=false
    Sampled Position Offset=&MaskVectorBySpawnGroup(Masked Vector=<0 0 -4>, Passthrough Vector=<-3 0 4>)
    Skeletal Mesh=$System.Skeletal Mesh
   )
   &DynamicMaterialParameters(
    Index 0 Param 1=&FloatFromCurve(FloatCurve=DataInterface() {propertyValues {Curve {keys [<at 0 0 RCIM_Cubic>, {arriveTangent=40.0499, interpMode=RCIM_Cubic, leaveTangent=40.0499, tangentMode=RCTM_User, time=1, value=12}]}}})
    Index 0 Param 2=&Multiply_Float(A=$Engine.Time, B=&RandomRangeFloat(Minimum=2.2, Maximum=3))
    Index 0 Param 4=&RandomRangeFloat(Minimum=0, Maximum=1)
    Module.Index0Float0Write=true
    Module.Index0Float1Write=true
    Module.Index0Float2Write=true
    Module.Index0Float3Write=true
   )
   &InterpolateOverTime(Rate Of Change=12, Target Vector Value:NiagaraPosition=$Particles.Position, Variable Type=::NewEnumerator2)
   /SetVariables(Particles.Interpolate DY:Vector4f=&MakeFloat4FromVectorAndFloat(
     "Vector (RGB)"=$Particles.InterpolateOverTime.InterpolateOverTime_V3_.Moving Average
     "Float (Alpha)"=&Lerp_Float(A=3, B=1.5, Alpha=1)
    )) {isSetParameters=true}
   &Timeline(
    Play=&SetBoolByFloatComparison(A=$NPC.NPC_Audio_Reactive_Parameters.Amplitude_Music, B=0.675)
    Rewind Rate=0.55
   )
   &ScaleColor(
    Scale Alpha=&FloatFromCurve(
     FloatCurve=DataInterface() {propertyValues {Curve {
        defaultValue=0
        keys [
         {interpMode=RCIM_Cubic, tangentMode=RCTM_User, time=0, value=0}
         {interpMode=RCIM_Cubic, tangentMode=RCTM_User, time=0.5, value=1}
         {tangentMode=RCTM_User, time=1, value=0}
        ]
       }}}
     Scale Curve=&Lerp_Float(
      A=&RandomRangeFloat(Minimum=0.12, Maximum=0.1575)
      B=&RandomRangeFloat(Minimum=0.4, Maximum=0.525)
      Alpha=&RemapRange(Input Value To Remap=$NPC.NPC_Audio_Reactive_Parameters.Amplitude_Music, Input Min=0.4, Input Max=0.925, Output Min=0, Output Max=1)
     )
    )
    ScaleA=true
    ScaleRGB=true
   )
   /SetVariables(
    Particles.BlackScale:Vector3f=&Multiply_Vector(A=$Particles.Scale, B=<1 1 1.2>)
    Particles.ShoulderBlackColor:LinearColor=&MultiplyLinearColor(
     Linear Color=&MakeLinearColorFromVectorAndFloat("Vector (RGB)"=<0 0 0>, "Float (Alpha)"=3)
     Scale Factor=&MakeVector4_FromLinearColor(LinearColor=$Particles.Color)
    )
   ) {isSetParameters=true}
  ]
  rendererClasses [&NiagaraMeshRendererProperties]
  renderers [
   {rendererClass=&NiagaraMeshRendererProperties, rendererIndex=0}
   {rendererClass=&NiagaraMeshRendererProperties, rendererIndex=1}
   {rendererClass=&NiagaraMeshRendererProperties, rendererIndex=2}
   {rendererClass=&NiagaraMeshRendererProperties, rendererIndex=3}
  ]
 }
 SphereFireBody {
  emitterUpdateScript [
   &EmitterState(
    Inactive Response=::NewEnumerator1
    Life Cycle Mode=::NewEnumerator1
    Loop Behavior=::NewEnumerator1
    Loop Duration Mode=::NewEnumerator1
    UseLoopDelay=false
   )
   &SpawnBurst_Instantaneous(Spawn Count=19, Use Spawn Probability=false, UseLoopCountLimit=false)
  ]
  particleSpawnScript [
   &PartitionParticles(Number Of Partitions=20)
   &InitializeParticle(
    Color=<rgba 0.6979 0.6979 0.6979 1>
    Lifetime=3.6
    Mesh Scale=&Multiply_VectorByFloat(Vector=<0.3 1 0.3>, Float=&RandomRangeFloat(Minimum=0.7, Maximum=0.8))
    Mesh Scale Mode=::NewEnumerator4
    Sprite Rotation Angle=-0.25
    Sprite Rotation Mode=::NewEnumerator1
    Sprite Size Mode=::NewEnumerator3
    Uniform Sprite Size=&Multiply_Float(
     A=&SelectFloatFromArray(
      Direct Array Index=$Particles.PartitionParticles.Partition
      Float Selection Array=DataInterface() {propertyValues {FloatData [0,25,25,20,20,20,20,18,18,20,20,30,30,25,25,25,25,20,20]}}
     )
     B=0.8
    )
    UsePositionOffset=false
    Write Material Random=true
   )
   /SetVariables(Particles.Random=&RandomRangeFloat(Minimum=0, Maximum=1)) {isSetParameters=true}
   &InitialMeshOrientation(Mesh Orientation Mode=::NewEnumerator2, Rotation=<0 0 0.25>, Use Rotation Vector=true)
   &InitialMeshOrientation(Mesh Orientation Mode=::NewEnumerator2, Rotation=<0.25 0 0>, Use Rotation Vector=true) {module=InitialMeshOrientation001}
   &CameraOffset(Camera Offset Amount=-8)
  ]
  particleUpdateScript [
   &ParticleState(Kill Particles When Lifetime Has Elapsed=false, Let Infinitely Lived Particles Die When Emitter Deactivates=true)
   &ScaleSpriteSize(
    Scale Sprite Size Mode=::NewEnumerator0
    Uniform Scale Factor=&MakeCustomFloatFromBool(
     Bool=$User.bisFootHidden
     True Float=&SelectFloatFromArray(
      Direct Array Index=$Particles.PartitionParticles.Partition
      Float Selection Array=DataInterface() {propertyValues {FloatData [0,1,1,1,1,1,1,1,1,1,1,1,1,1,1,0,0,0,0]}}
     )
     False Float=1
    )
   ) {module=ScaleSpriteSize001}
   &SkeletalMeshLocation(
    Bone / Socket Index=&SelectIntFromArray(
     Direct Array Index=$Particles.PartitionParticles.Partition
     Int Selection Array=DataInterface() {propertyValues {IntData [0,1,2,1,2,5,5,7,8,6,6,9,10,9,10,13,14,13,14]}}
    )
    Bone Sampling Mode=::NewEnumerator4
    MaskSpawnGroup=false
    Sampled Position Offset=&SelectVectorFromArray(
     Direct Array Index=$Particles.PartitionParticles.Partition
     Vector Selection Array=DataInterface() {propertyValues {FloatData [
        <0 0 0>
        <0 0 0>
        <0 0 0>
        <13 0 0>
        <-13 0 0>
        <0 0 0>
        <0 0 0>
        <0 0 0>
        <0 0 0>
        <0 0 0>
        <0 0 0>
        <0 0 0>
        <0 0 0>
        <-20 0 0>
        <20 0 0>
        <0 0 0>
        <0 0 0>
        <-20 0 0>
        <20 0 0>
       ]}}
    )
    Skeletal Mesh=$System.Skeletal Mesh
   )
   &SolveForcesAndVelocity(Clamp Velocity=false, Limit Acceleration=false)
   &Timeline(Play=&SetBoolByFloatComparison(A=$NPC.NPC_Audio_Reactive_Parameters.Amplitude_Music, B=0.675), Rewind Rate=0.6)
   &ScaleColor(
    Scale Alpha=&Lerp_Float(
     A=&RandomRangeFloat(Minimum=0.15, Maximum=0.24)
     B=&RandomRangeFloat(Minimum=0.5, Maximum=0.8)
     Alpha=&RemapRange(Input Value To Remap=$NPC.NPC_Audio_Reactive_Parameters.Amplitude_Music, Input Min=0.4, Input Max=0.925, Output Min=0, Output Max=1)
    )
    Scale RGB=&MakeVectorFromLinearColorRGB(Color=&RandomRangeLinearColor(Minimum=<rgba 0.3333 0.0226 0.0589 1>, Maximum=<rgba 0.5365 0.0503 0.1072 1>))
    ScaleA=true
    ScaleRGB=true
   )
   &ScaleSpriteSize(Scale Sprite Size Mode=::NewEnumerator0, Uniform Scale Factor=2)
   &ScaleColor(
    Scale Alpha=&SelectFloatFromArray(
     Direct Array Index=$Particles.PartitionParticles.Partition
     Float Selection Array=DataInterface() {propertyValues {FloatData [1,1,1,1,0,0,0,0,0,0,0,0.4,0.4,0.4,0.4,0.4,1,1,0]}}
    )
    ScaleA=true
    ScaleRGB=true
   ) {module=ScaleColor001}
  ]
  rendererClasses [&NiagaraSpriteRendererProperties]
  renderers [{rendererClass=&NiagaraSpriteRendererProperties}]
 }
}
summary {
 systemName=NS_Character_StuffRasp_Veil_BoneFire
 emitters [
  {emitterName=FireSwirls_Calf, rendererClasses [&NiagaraMeshRendererProperties]}
  {emitterName=FireBodyStatic, rendererClasses [&NiagaraMeshRendererProperties]}
  {emitterName=SphereFireBody, rendererClasses [&NiagaraSpriteRendererProperties]}
  {emitterName=FireSwirls_ArmThigh, rendererClasses [&NiagaraMeshRendererProperties]}
  {emitterName=FireSwirls_Pelvis, rendererClasses [&NiagaraMeshRendererProperties]}
  {emitterName=Fire_Eyes, rendererClasses [&NiagaraMeshRendererProperties]}
  {emitterName=Minimal, rendererClasses [&NiagaraMeshRendererProperties]}
 ]
 userVariables [
  User.bisFootHidden {type=NiagaraBool}
  User.bIsFrontEnd {type=NiagaraBool}
  User.bIsFrontEndPreview {type=NiagaraBool}
  User.bisHandsHidden {type=NiagaraBool}
  User.bisHeadHidden {type=NiagaraBool}
 ]
}
