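# Parent material for the StuffRasp Maze character's face accessory: masked, unlit and two-sided, with a near-black base color.
# Opacity keeps only back faces, cut by an inverted body mask texture, while the shadow pass stays fully opaque.
# World position offset pulls camera-facing vertices inward along normals, with depth-scaled bias between two parameters.
# These overrides sit between the shared CharacterParent PreFX and PostFX functions.
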Material /BRCosmetics/Characters/Player/Male/Medium/Bodies/M_MED_StuffRasp_Maze/Material/M_StuffRasp_Maze_FaceAcc_Parent
graph {
 properties {
  bAutomaticallySetUsageInEditor=false
  bUseMaterialAttributes=true
  bUsedWithMeshDeformer=true
  bUsedWithMorphTargets=true
  bUsedWithSkeletalMesh=true
  blendMode=BLEND_Masked
  pixelDepthOffsetMode=PDOM_Legacy
  shadingModel=MSM_Unlit
  twoSided=true
  usedShadingModels=Unlit
 }
 nodes [
  Constant()
  n1 = TextureSample() {
   samplerSource=SSM_Wrap_WorldGroupSettings
   samplerType=SAMPLERTYPE_Masks
   texture=/BRCosmetics/Characters/Player/Male/Medium/Bodies/M_MED_StuffRasp/Textures/T_StuffRasp_Body_TAMask2
  }
  n2 = OneMinus(None=Saturate(None=Multiply(A=@n1.R) {constB=2}))
  n3 = LinearInterpolate(A=Constant() {r=1}, Alpha=Saturate(None=TwoSidedSign())) {constB=0}
  n4 = DotProduct(A=Subtract(A=CameraPositionWS(), B=WorldPosition().XYZ), B=VertexNormalWS())
  n5 = NamedRerouteDeclaration(None=Saturate(None=Multiply(A=@n4) {constB=5})) {name=FacingMask, nodeColor=<rgba 0 1 0.6719 1>}
  n6 = PixelDepth()
  n7 = LinearInterpolate(
   A=ScalarParameter() {defaultValue=-0.02, parameterName=WPO_Shrink_Bias_Min}
   B=ScalarParameter() {defaultValue=-0.06, parameterName=WPO_Shrink_Bias_Max}
   Alpha=Saturate(None=Divide(A=@n6.R) {constB=300})
  ) {constB=-0.06}
  n8 = SetMaterialAttributes!(
   MaterialAttributes=NamedRerouteUsage()
   Base Color=Constant4Vector() {constant=<rgba 0.01 0.01 0.01 1>}.RGB
   Metallic
   Specular=ScalarParameter()
   Emissive Color
   Opacity Mask=ShadowReplace(Default=Multiply(A=@n3, B=@n2), Shadow=Constant() {r=1})
   World Position Offset=Multiply(A=VertexNormalWS(), B=Multiply(A=@n5, B=@n7))
  ) {attributeSetTypes [
    #1
    #2
    #3
    #4
    #5
    #6
   ]}
  n9 = MaterialFunctionCall() {materialFunction=&MF_CharacterParent_2023_PreFX}
  MaterialFunctionCall(Material=@n8, M=@n9.M, Pre Skinned Local Position=@n9.M, Distance Blend=@n9.M, Fresnel=@n9.M) {materialFunction=&MF_CharacterParent_2023_PostFX}
  NamedRerouteDeclaration(None=@n9.Result) {name=IN_MA, nodeColor=<rgba 0 1 0.4609 1>}
 ]
}
