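# Translucent, two-sided parent material for the StuffRasp Maze character's face accessory overlay, built between the shared character PreFX and PostFX functions.
# A Fresnel ramp drives base color, while a mask texture supplies specular and roughness.
# Opacity combines texture masks, Fresnel and panning noise, alongside a procedural noise-modulated dot pattern made from tiled sphere masks.
# TwoSidedSign zeroes back-face opacity, and a SortTranslucency function processes the final opacity.
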Material /BRCosmetics/Characters/Player/Male/Medium/Bodies/M_MED_StuffRasp_Maze/Material/M_StuffRasp_Maze_FaceAccoverlay_Parent
graph {
 properties {
  allowTranslucentCustomDepthWrites=true
  bAutomaticallySetUsageInEditor=false
  bCastRayTracedShadows=false
  bUseMaterialAttributes=true
  bUsedWithMeshDeformer=true
  bUsedWithSkeletalMesh=true
  blendMode=BLEND_Translucent
  ditherOpacityMask=true
  translucencyLightingMode=TLM_Surface
  translucencyPass=MTP_BeforeDOF
  twoSided=true
  usedShadingModels=Default Lit
 }
 nodes [
  Constant() {r=0.5}
  Constant() {r=1}
  n1 = SetMaterialAttributes!(
   MaterialAttributes=NamedRerouteUsage()
   Base Color=NamedRerouteUsage()
   Metallic=Constant()
   Specular=NamedRerouteUsage()
   Roughness=NamedRerouteUsage()
   Anisotropy
   Emissive Color
   Opacity=MaterialFunctionCall!(Opacity=NamedRerouteUsage()) {materialFunction=/Game/Global/Materials/MaterialGym/Materials/SortedOutline/MF_SortTranslucency}.Result
  ) {attributeSetTypes [
    #1
    #2
    #3
    #4
    #5
    #6
    #7
   ]}
  n2 = MaterialFunctionCall() {materialFunction=&MF_CharacterParent_2023_PreFX}
  MaterialFunctionCall(Material=@n1, M=@n2.M, Pre Skinned Local Position=@n2.M, Distance Blend=@n2.M, Fresnel=@n2.M) {materialFunction=&MF_CharacterParent_2023_PostFX}
  n3 = TextureSample(UVs=Multiply(A=TextureCoordinate()) {constB=2}) {samplerSource=SSM_Wrap_WorldGroupSettings, samplerType=SAMPLERTYPE_Grayscale, texture=&LowResBlurredNoise}
  n4 = TextureSample() {
   samplerSource=SSM_Wrap_WorldGroupSettings
   samplerType=SAMPLERTYPE_Masks
   texture=/BRCosmetics/Characters/Player/Male/Medium/Bodies/M_MED_StuffRasp/Textures/T_StuffRasp_FaceAccExtra_M
  }
  n5 = Power(Base=@n4.G) {constExponent=4}
  n6 = Multiply(A=Multiply(A=@n3.R, B=Constant() {r=0.7}) {constB=0.5}, B=@n5)
  n7 = MaterialFunctionCall(UVs=TextureCoordinate(), "Rotation Angle (0-1)"=ScalarParameter() {defaultValue=0.1, parameterName=Dot_Rotation}) {materialFunction=&CustomRotator}
  n8 = ScalarParameter() {defaultValue=320, group=StuffRasp_axe, parameterName=Dot_Tiling}
  n9 = SphereMask(
   A=Frac(None=Multiply(A=@n7.Rotated Values, B=@n8))
   B=Constant() {r=0.5}
   Radius=LinearInterpolate(
    A=ScalarParameter() {defaultValue=0.7762, group=StuffRasp_axe, parameterName=Dot_RadiusMax}
    B=ScalarParameter() {defaultValue=0.276, group=StuffRasp_axe, parameterName=Dot_RadiusMin}
    Alpha=Saturate(None=@n5)
   )
   Hardness=ScalarParameter() {defaultValue=0.9, parameterName=Dot_Hardness}
  )
  NamedRerouteDeclaration(None=Multiply(A=Multiply(A=@n6, B=@n9), B=NamedRerouteUsage())) {name=DotMask, nodeColor=<rgba 0.5469 1 0 1>}
  NamedRerouteDeclaration(None=MaterialFunctionCall!(
    TriplanarUV_UseClamp?=ScalarParameter() {group=TriplanarUV, parameterName=TriplanarUV_3D_UseClamp}
    TriplanarUV_FlipUV=ScalarParameter() {group=TriplanarUV, parameterName=TriplanarUV_3D_FlipUV}
    TriplanarUV_CenterPosition=VectorParameter() {group=TriplanarUV, parameterName=TriplanarUV_3D_CenterPosition, defaultValue=<rgba 0 0 0 1>}.RGB
    TriplanarUV_UVScale=ScalarParameter() {defaultValue=82, group=TriplanarUV, parameterName=TriplanarUV_3D_UVSize}
   ) {materialFunction=/BRCosmetics/Characters/Player/Male/Medium/Bodies/M_MED_BirchPalace/Skins/Gas/Materials/MF_TriplanarUV}.UV_FrontProjection) {name=TriplannarUV_front, nodeColor=<rgba 1 0.3516 0 1>}
  n10 = NamedRerouteUsage()
  n11 = TextureSample() {
   samplerSource=SSM_Wrap_WorldGroupSettings
   samplerType=SAMPLERTYPE_Masks
   texture=&T_StuffRasp_FaceAccExtra_M
  }
  n12 = OneMinus(None=@n11.R)
  n13 = Constant() {r=0.5}
  n14 = ScalarParameter() {defaultValue=0.9, group=StuffRasp, parameterName=Global transparency}
  n15 = Constant() {r=0.04}
  n16 = Multiply(A=Power(Base=Fresnel(ExponentIn=@n15), Exp=Constant() {r=3}), B=Constant() {r=1})
  n17 = TextureSample() {
   samplerSource=SSM_Wrap_WorldGroupSettings
   samplerType=SAMPLERTYPE_Masks
   texture=/BRCosmetics/Characters/Player/Male/Medium/Bodies/M_MED_StuffRasp/Textures/T_StuffRasp_Body_TAMask2
  }
  n18 = OneMinus(None=@n17.R)
  n19 = Multiply(
   A=Add(A=Multiply(A=@n12, B=@n13), B=LinearInterpolate(B=@n14, Alpha=@n16) {constB=0.9})
   B=NamedRerouteDeclaration(None=Multiply(A=@n18, B=@n17.G)) {name=HideMask, nodeColor=<rgba 0.7109 1 0 1>}
  )
  n20 = TextureSample(UVs=Panner(Coordinate=NamedRerouteUsage()) {speedY=0.03}) {samplerSource=SSM_Wrap_WorldGroupSettings, samplerType=SAMPLERTYPE_Grayscale, texture=&LowResBlurredNoise}
  n21 = Constant() {r=2}
  n22 = Saturate(None=Power(Base=Multiply(A=@n20.R, B=@n21), Exp=Constant() {r=2}))
  n23 = Constant() {r=0.1}
  n24 = Add(A=@n19, B=Multiply(A=@n19, B=Multiply(A=@n22, B=@n23)))
  n25 = TwoSidedSign()
  NamedRerouteDeclaration(None=LinearInterpolate(A=Constant(), B=Add(A=@n10, B=@n24), Alpha=Saturate(None=@n25)) {constA=1, constB=0}) {name=OutOpacity, nodeColor=<rgba 1 0 0.2578 1>}
  n26 = Fresnel(ExponentIn=ScalarParameter() {defaultValue=3, group=StuffRasp, parameterName=StuffRasp_FresnelExponentln})
  n27 = ScalarParameter() {defaultValue=1.5, parameterName=StuffRasp_FresnelPower}
  NamedRerouteDeclaration(None=LinearInterpolate(
    A=ScalarParameter() {defaultValue=0.01, group=StuffRasp, parameterName=StuffRasp_FresnelMin}
    B=ScalarParameter() {defaultValue=0.35, group=StuffRasp, parameterName=StuffRasp_FresnelMax}
    Alpha=Power(Base=@n26, Exp=@n27)
   )) {name=OutBaseColor, nodeColor=<rgba 0.7813 0 1 1>}
  n28 = TextureSample() {
   samplerSource=SSM_Wrap_WorldGroupSettings
   samplerType=SAMPLERTYPE_Masks
   texture=/BRCosmetics/Characters/Player/Male/Medium/Bodies/M_MED_StuffRasp/Textures/T_StuffRasp_FaceAccExtra_S
  }
  NamedRerouteDeclaration(None=Multiply(A=@n28.R, B=Constant() {r=0.2})) {name=OutSpecular, nodeColor=<rgba 0 1 0.4609 1>}
  NamedRerouteDeclaration(None=@n28.G) {name=Out Rouhness, nodeColor=<rgba 1 0.1406 0 1>}
  NamedRerouteDeclaration(None=@n2.Result) {name=InMa, nodeColor=<rgba 0 0.1797 1 1>}
 ]
}
