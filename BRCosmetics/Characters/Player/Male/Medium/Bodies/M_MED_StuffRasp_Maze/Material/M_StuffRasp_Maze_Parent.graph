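# Parent material for the StuffRasp Maze character body: masked, two-sided and unlit, drawing black back faces only.
# Opacity multiplies a back-face test by an inverted, channel-selected mask texture, while ShadowReplace keeps shadows fully opaque.
# World position offset pulls camera-facing vertices inward along their normals, with shrink bias blended by pixel depth.
# These attributes are set between the shared CharacterParent 2023 PreFX and PostFX functions.
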Material /BRCosmetics/Characters/Player/Male/Medium/Bodies/M_MED_StuffRasp_Maze/Material/M_StuffRasp_Maze_Parent
graph {
 properties {
  bAutomaticallySetUsageInEditor=false
  bUseMaterialAttributes=true
  bUsedWithMeshDeformer=true
  bUsedWithMorphTargets=true
  bUsedWithSkeletalMesh=true
  blendMode=BLEND_Masked
  pixelDepthOffsetMode=PDOM_Legacy
  shadingModel=MSM_Unlit
  twoSided=true
  usedShadingModels=Unlit
 }
 nodes [
  n1 = Constant()
  n2 = LinearInterpolate(Alpha=Saturate(None=TwoSidedSign())) {constA=1, constB=0}
  n3 = TextureSampleParameter2D() {
   group=StuffRasp
   parameterName=StuffRasp_OpacityMask
   samplerType=SAMPLERTYPE_Masks
   texture=/BRCosmetics/Characters/Player/Male/Medium/Bodies/M_MED_StuffRasp/Textures/T_StuffRasp_Body_TAMask1
  }
  n4 = Multiply(A=ChannelMaskParameter(None=@n3.RGB) {group=StuffRasp, parameterName=StuffRasp_ChannelSelect}.R) {constB=2}
  n5 = OneMinus(None=NamedRerouteDeclaration(None=Saturate(None=@n4)) {name=OpacityMask, nodeColor=<rgba 0.2656 1 0 1>})
  n6 = DotProduct(A=Subtract(A=CameraPositionWS(), B=WorldPosition().XYZ), B=VertexNormalWS())
  n7 = NamedRerouteDeclaration(None=Saturate(None=Multiply(A=@n6) {constB=5})) {name=FacingMask, nodeColor=<rgba 0 1 0.6719 1>}
  n8 = PixelDepth()
  n9 = LinearInterpolate(
   A=ScalarParameter() {defaultValue=-0.02, parameterName=WPO_Shrink_Bias_Min}
   B=ScalarParameter() {defaultValue=-0.06, parameterName=WPO_Shrink_Bias_Max}
   Alpha=Saturate(None=Divide(A=@n8.R) {constB=300})
  ) {constB=-0.06}
  n10 = SetMaterialAttributes!(
   MaterialAttributes=NamedRerouteUsage()
   Base Color=VectorParameter() {group=StuffRasp, parameterName=StuffRasp_BaseColor, defaultValue=<rgba 0.01 0.01 0.01 1>}.RGB
   Metallic=@n1
   Specular=@n1
   Emissive Color=@n1
   Opacity Mask=ShadowReplace(Default=Multiply(A=@n2, B=@n5), Shadow=Constant() {r=1})
   Shading Model
   Roughness=Constant() {r=1}
   World Position Offset=Multiply(A=VertexNormalWS(), B=Multiply(A=@n7, B=@n9))
   Pixel Depth Offset
  ) {attributeSetTypes [
    #1
    #2
    #3
    #4
    #5
    #6
    #7
    #8
    #9
   ]}
  n11 = MaterialFunctionCall() {materialFunction=&MF_CharacterParent_2023_PreFX}
  MaterialFunctionCall(Material=@n10, M=@n11.M, Pre Skinned Local Position=@n11.M, Distance Blend=@n11.M, Fresnel=@n11.M) {materialFunction=&MF_CharacterParent_2023_PostFX}
  NamedRerouteDeclaration(None=@n11.Result) {name=IN_MA, nodeColor=<rgba 0 1 0.4609 1>}
 ]
}
