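# Material function in the StuffRasp Maze character body that outputs an object-relative view-depth gradient from pixel depth.
# Named reroutes define object center depth as camera-to-object distance, with near and far depths offset by the Z component of the object bounds.
# A normalized depth difference is saturated and inverted, then shaped by a contrast exponent.
# The parameter named Offset scales the result as a multiplier rather than adding to it.
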MaterialFunction /BRCosmetics/Characters/Player/Male/Medium/Bodies/M_MED_StuffRasp_Maze/Material/MF_ObjectViewDepth
graph {nodes [
  n1 = Subtract(A=PixelDepth().R, B=NamedRerouteUsage())
  n2 = Subtract(A=NamedRerouteUsage(), B=NamedRerouteUsage())
  n3 = OneMinus(None=Saturate(None=Divide(A=@n1, B=@n2)))
  n4 = ScalarParameter() {defaultValue=3.4178, group=ObjectViewDepth, parameterName=ObjectViewDepth01Contrast}
  FunctionOutput(None=Multiply(A=Power(Base=@n3, Exp=@n4), B=ScalarParameter() {defaultValue=3.7029, group=ObjectViewDepth, parameterName=ObjectViewDepth01Offset}))
  NamedRerouteDeclaration(None=Distance(A=CameraPositionWS(), B=ObjectPositionWS())) {name=ObjectCenterDepth, nodeColor=<rgba 0.9453 1 0 1>}
  n5 = ComponentMask(None=ObjectBounds()) {b=true}
  NamedRerouteDeclaration(None=Subtract(A=NamedRerouteUsage(), B=@n5)) {name=ObjectNearDepth, nodeColor=<rgba 0.7344 0 1 1>}
  NamedRerouteDeclaration(None=Add(A=NamedRerouteUsage(), B=@n5)) {name=ObjectfarDepth, nodeColor=<rgba 0 1 0.4141 1>}
 ]}
