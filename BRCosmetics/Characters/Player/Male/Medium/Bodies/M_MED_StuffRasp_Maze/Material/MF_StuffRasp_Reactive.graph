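# Material function for the StuffRasp Maze character body that turns an input into an audio-reactive result, outputting OutReactive and AmplitudeReactive.
# Audio amplitude comes from MF_AudioReactivity_FrontendLoop, with a curve-atlas fake loop for the frontend, toggled by EnableFrontendFakeAudioLoop.
# Pre-skinned local position, remapped as BoneDepth, supports sine-band animation in two layers, one time-driven and one with switchable offset.
# Named reroutes connect these bands, panning noise, triplanar UVs and amplitude-scaled intensity.
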MaterialFunction /BRCosmetics/Characters/Player/Male/Medium/Bodies/M_MED_StuffRasp_Maze/Material/MF_StuffRasp_Reactive
graph {nodes [
  n1 = FunctionInput() {bUsePreviewValueAsDefault=true}
  n2 = Add(A=Multiply(A=@n1) {constB=0.7}, B=Multiply(A=@n1, B=NamedRerouteUsage()))
  n3 = LinearInterpolate(A=Constant() {r=0.6}, B=Constant() {r=5}, Alpha=NamedRerouteUsage()) {constB=2}
  n4 = ScalarParameter() {defaultValue=1, group=StuffRasp_Reactive, parameterName=ReactiveMin}
  n5 = ScalarParameter() {defaultValue=5, group=StuffRasp_Reactive, parameterName=ReactiveMax}
  n6 = ScalarParameter() {desc=Set this true in the part modifier BP, group=Audio Reactivity, parameterName=EnableFrontendFakeAudioLoop, sortPriority=200}
  n7 = LinearInterpolate(A=NamedRerouteUsage(), B=NamedRerouteUsage(), Alpha=@n6)
  FunctionOutput(None=Multiply(A=Multiply(A=@n2, B=@n3), B=LinearInterpolate(A=@n4, B=@n5, Alpha=@n7))) {outputName=OutReactive}
  FunctionOutput(None=NamedRerouteDeclaration(None=LinearInterpolate(Alpha=@n7) {constB=3}) {name=AmplitudeReactive, nodeColor=<rgba 1 0 0.2109 1>}) {outputName=AmplitudeReactive}
  NamedRerouteDeclaration(None=MaterialFunctionCall!(
    TriplanarUV_UseClamp?=ScalarParameter() {group=TriplanarUV, parameterName=TriplanarUV_3D_UseClamp}
    TriplanarUV_FlipUV=ScalarParameter() {group=TriplanarUV, parameterName=TriplanarUV_3D_FlipUV}
    TriplanarUV_CenterPosition=VectorParameter() {group=TriplanarUV, parameterName=TriplanarUV_3D_CenterPosition, defaultValue=<rgba 0 0 0 1>}.RGB
    TriplanarUV_UVScale=ScalarParameter() {defaultValue=82, group=TriplanarUV, parameterName=TriplanarUV_3D_UVSize}
   ) {materialFunction=/BRCosmetics/Characters/Player/Male/Medium/Bodies/M_MED_BirchPalace/Skins/Gas/Materials/MF_TriplanarUV}.UV_FrontProjection) {name=TriplannarUV_front, nodeColor=<rgba 1 0.3516 0 1>}
  n8 = ScalarParameter() {parameterName=FakeAudioTime2}
  n9 = ScalarParameter() {defaultValue=0.75, group=Audio Reactivity, parameterName=FakeAudioCurveSpeed}
  n10 = CurveAtlasRowParameter(CurveTime=Frac(None=Multiply(A=@n8, B=@n9))) {
   atlas=&/CA_Frontend_FakeAudioReactivity_StuffRasp
   curve=&/C_Frontend_FakeAudioCurve_StuffRasp
   group=Audio Reactivity
   parameterName=FakeAudioCurveAtlasIndex2
  }
  NamedRerouteDeclaration(None=Multiply(A=AppendVector(A=@n10, B=@n10), B=ScalarParameter() {defaultValue=1, group=Audio Reactivity, parameterName=FakeAudioReactivityMaxIntensity})) {name=CustomFakeAudio, nodeColor=<rgba 1 0.2578 0 1>}
  n11 = ChannelMaskParameter(None=VertexInterpolator(VS=LocalPosition() {localOrigin=InstancePreSkinning}.XYZ).PS) {group=StuffRasp_Reactive, maskChannel=Green, parameterName=BoneDepthChannel, defaultValue=<rgba 0 1 0 0>}
  NamedRerouteDeclaration(None=MaterialFunctionCall(
    Input=Subtract(A=@n11.R) {constB=0}
    Input Low=ScalarParameter() {defaultValue=-20, group=StuffRasp_Reactive, parameterName=BoneDepth_InputLow}
    Input High=ScalarParameter() {defaultValue=20, group=StuffRasp_Reactive, parameterName=BoneDepth_InputHigh}
    Target Low=Constant()
    Target High=ScalarParameter() {defaultValue=3.15, group=StuffRasp_Reactive, parameterName=BoneDepth_TargetHigh}
   ) {materialFunction=&RemapValueRange}.Result) {name=BoneDepth, nodeColor=<rgba 1 0 0.0234 1>}
  n12 = TextureSample(UVs=Panner(Coordinate=NamedRerouteUsage()) {speedY=0.02}) {samplerSource=SSM_Wrap_WorldGroupSettings, samplerType=SAMPLERTYPE_Grayscale, texture=&LowResBlurredNoise}
  n13 = Multiply(A=@n12.R, B=Constant() {r=2})
  n14 = Constant() {r=4}
  NamedRerouteDeclaration(None=Saturate(None=Power(Base=@n13, Exp=@n14))) {name=Noise, nodeColor=<rgba 1 0.2109 0 1>}
  n15 = Constant4Vector() {constant=<rgba 0.8 0.9002 1 1>}
  n16 = Constant() {r=0.6}
  n17 = If(A=NamedRerouteUsage(), "A > B"=Constant() {r=2}, "A == B"=@n16, "A < B"=@n16) {constB=0.01}
  n18 = NamedRerouteUsage()
  n19 = ComponentMask(None=NamedRerouteUsage()) {g=true}
  n20 = LinearInterpolate(Alpha=Saturate(None=OneMinus(None=@n19))) {constB=2}
  n21 = ScalarParameter() {desc=Set this true in the part modifier BP, group=Audio Reactivity, parameterName=EnableFrontendFakeAudioLoop, sortPriority=200}
  n22 = Constant()
  n23 = If(A=@n21, "A > B"=Constant() {r=1}, "A == B"=@n22, "A < B"=@n22) {constB=0.5}
  n24 = ScalarParameter() {defaultValue=-1, group=StuffRasp_Reactive, parameterName=ReactiveDirection}
  n25 = Multiply(
   A=LinearInterpolate(
    A=ScalarParameter() {defaultValue=-0.1, group=StuffRasp_Reactive, parameterName=ReactiveAnim_Layer2}
    B=Subtract(A=@n20) {constB=0.2}
    Alpha=@n23
   )
   B=@n24
  ) {constB=2}
  n26 = NamedRerouteUsage()
  n27 = NamedRerouteDeclaration(None=Sine(None=Add(A=@n25, B=@n26)) {period=3.14}) {name=Layer2, nodeColor=<rgba 1 0 0.7266 1>}
  n28 = Power(
   Base=SphereMask(
    A=Saturate(None=@n27)
    B=Constant() {r=0.993}
    Radius=ScalarParameter() {defaultValue=0.2, group=StuffRasp_Reactive, parameterName=ReactiveRadius}
   ) {hardnessPercent=0}
   Exp=Constant() {r=6.6942}
  )
  n29 = Multiply(A=Time(), B=@n24) {constB=2}
  n30 = NamedRerouteUsage()
  n31 = NamedRerouteDeclaration(None=Sine(None=Add(A=@n29, B=@n30)) {period=3.14}) {name=Layer1, nodeColor=<rgba 0.4062 1 0 1>}
  n32 = Multiply(A=Saturate(None=Power(Base=@n31) {constExponent=30}), B=OneMinus(None=NamedRerouteUsage()))
  NamedRerouteDeclaration(None=Multiply(A=Multiply(A=@n15.RGB, B=@n17), B=If(A=@n18, "A > B"=@n28, "A == B"=@n32, "A < B"=@n32) {constB=0.01})) {name=ReactiveAnim, nodeColor=<rgba 0 0.1094 1 1>}
  NamedRerouteDeclaration(None=@n23) {name=FakeAudio_switch, nodeColor=<rgba 0.1953 1 0 1>}
  n33 = MaterialFunctionCall!(InAmplitude, InSpectrum, CustomFakeSpectrum, CustomFakeAmplitude=ComponentMask(None=NamedRerouteUsage()) {r=true}) {materialFunction=/BRCosmetics/Global/MaterialFunctions/AudioReactivity/MF_AudioReactivity_FrontendLoop}
  n34 = ScalarParameter() {defaultValue=1, group=StideMIce_Reactive, parameterName=EnableAudioReactivity, sliderMax=1}
  NamedRerouteDeclaration(None=Power(
    Base=Multiply(A=@n33.Amplitude, B=@n34)
    Exp=ScalarParameter() {defaultValue=3, group=StideMIce_Reactive, parameterName=AudioReactivityPower}
   ) {constExponent=5}) {name=Amplitude, nodeColor=<rgba 0 1 0.1094 1>}
  NamedRerouteDeclaration(None=@n33.Amplitude) {name=FakeAmplitude, nodeColor=<rgba 0 0.5312 1 1>}
 ]}
