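# Parent material for the StuffRasp Maze player character body, layering custom reactive emissive between the shared CharacterParent 2023 PreFX and PostFX functions.
# A static switch, on by default, applies a masked Fresnel curve remap to the PreFX attributes.
# The resulting emissive is scaled by object view depth, desaturated and passed through the StuffRasp Reactive function.
# A pre-skinned-position sphere mask dims the glow around the leg center before an emissive tint is applied.
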Material /BRCosmetics/Characters/Player/Male/Medium/Bodies/M_MED_StuffRasp_Maze/Material/M_StuffRasp_Maze_XParent
graph {
 properties {
  bUseMaterialAttributes=true
  bUsedWithMeshDeformer=true
  bUsedWithSkeletalMesh=true
  pixelDepthOffsetMode=PDOM_Legacy
  shadingModel=MSM_FromMaterialExpression
  usedShadingModels=Default Lit | Subsurface | Clear Coat | Subsurface Profile | Cloth
 }
 nodes [
  n1 = MaterialFunctionCall() {materialFunction=&MF_CharacterParent_2023_PreFX}
  n2 = TextureObjectParameter() {
   group=RemapCurveByFresnel
   parameterName=TechArtMaskRe
   samplerType=SAMPLERTYPE_Masks
   texture=&T_WhiteMask
  }
  n3 = GetMaterialAttributes(None=StaticSwitchParameter(
    True=MaterialFunctionCall!(MA=@n1.Result, FXMask=@n2, PreskinLocalPosition=@n1.Result) {materialFunction=&/MF_RemapCurveByFresnel_StuffRasp_Body}.Result
    False=@n1.Result
   ) {defaultValue=true, group=Skin FX, parameterName=UseRemapCurveByFresnel}) {attributeGetTypes [#1]}
  n4 = MaterialFunctionCall() {materialFunction=&/MF_ObjectViewDepth}
  n5 = Multiply(A=@n3.Emissive Color, B=LinearInterpolate(A=Constant(), B=Constant() {r=3.85}, Alpha=Saturate(None=@n4.Result)))
  n6 = Constant() {r=1}
  n7 = NamedRerouteUsage()
  n8 = VectorParameter() {group=StuffRasp, parameterName=StuffRasp_LegCenterPosition, defaultValue=<rgba 0 0 0 1>}
  n9 = ScalarParameter() {defaultValue=56.6545, group=StuffRasp, parameterName=StuffRasp_LegMaskRadius}
  n10 = ScalarParameter() {defaultValue=0.576, parameterName=StuffRasp_LegMaskHardness}
  n11 = Multiply(
   A=MaterialFunctionCall!(In=Desaturation(None=@n5, Fraction=@n6)) {materialFunction=&/MF_StuffRasp_Reactive}.OutReactive
   B=LinearInterpolate(Alpha=SphereMask(A=@n7, B=@n8.RGB, Radius=@n9, Hardness=@n10)) {constA=1, constB=0.02}
  )
  n12 = VectorParameter() {group=StuffRasp, parameterName=Emissive_Tint, defaultValue=<rgba 1 1 1 1>}
  MaterialFunctionCall(
   Material=SetMaterialAttributes(MaterialAttributes=@n3.MaterialAttributes, Emissive Color=Multiply(A=@n11, B=@n12.RGB)) {attributeSetTypes [#1]}
   M=@n1.M
   Pre Skinned Local Position=@n1.M
   Distance Blend=@n1.M
   Fresnel=@n1.M
  ) {materialFunction=&MF_CharacterParent_2023_PostFX}
  NamedRerouteDeclaration(None=@n1.Pre skinned Local Position) {name=PreSkinnedPosition, nodeColor=<rgba 0 0.3437 1 1>}
 ]
}
