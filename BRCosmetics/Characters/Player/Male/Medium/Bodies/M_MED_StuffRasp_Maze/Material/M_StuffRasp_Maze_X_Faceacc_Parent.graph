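# Parent material for the face accessory of the StuffRasp Maze character body, built between the shared 2023 character PreFX and PostFX functions.
# A fresnel curve-remap function feeds emissive, which is boosted, desaturated, passed through a reactive function and masked.
# Brightness parameters and AO masks scale that glow, then a powered mask channel adds highlights and a tint colors the result.
# Only emissive is overridden within the material attributes.
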Material /BRCosmetics/Characters/Player/Male/Medium/Bodies/M_MED_StuffRasp_Maze/Material/M_StuffRasp_Maze_X_Faceacc_Parent
graph {
 properties {
  bUseMaterialAttributes=true
  bUsedWithMeshDeformer=true
  bUsedWithMorphTargets=true
  bUsedWithSkeletalMesh=true
  pixelDepthOffsetMode=PDOM_Legacy
  shadingModel=MSM_FromMaterialExpression
  usedShadingModels=Default Lit | Subsurface | Clear Coat | Subsurface Profile | Cloth
 }
 nodes [
  n1 = MaterialFunctionCall() {materialFunction=&MF_CharacterParent_2023_PreFX}
  n2 = TextureObjectParameter() {
   group=RemapCurveByFresnel
   parameterName=TechArtMaskRe
   samplerType=SAMPLERTYPE_Masks
   texture=&T_WhiteMask
  }
  n3 = MaterialFunctionCall!(MA=@n1.Result, FXMask=@n2, PreskinLocalPosition=@n1.Result) {materialFunction=&/MF_RemapCurveByFresnel_StuffRasp_FaceAcc}
  n4 = GetMaterialAttributes(None=@n3.Result) {attributeGetTypes [#1]}
  n5 = TextureSample() {samplerType=SAMPLERTYPE_Masks, texture=/BRCosmetics/Characters/Player/Male/Medium/Bodies/M_MED_StuffRasp/Textures/T_StuffRasp_Body_TAMask3}
  n6 = Multiply(A=@n4.Emissive Color) {constB=3}
  n7 = Constant() {r=1}
  n8 = TextureSample() {samplerType=SAMPLERTYPE_Masks, texture=/BRCosmetics/Characters/Player/Male/Medium/Bodies/M_MED_StuffRasp/Textures/T_StuffRasp_FaceAcc_M}
  n9 = Multiply(
   A=MaterialFunctionCall!(In=Desaturation(None=@n6, Fraction=@n7)) {materialFunction=&/MF_StuffRasp_Reactive}.OutReactive
   B=@n8.R
  )
  n10 = TextureSampleParameter2D() {
   parameterName=FaceAcc_AO
   samplerType=SAMPLERTYPE_Masks
   texture=/BRCosmetics/Characters/Player/Male/Medium/Bodies/M_MED_StuffRasp/Textures/T_StuffRasp_Body_TAMask2
  }
  n11 = LinearInterpolate(
   A=LinearInterpolate(A=ScalarParameter() {defaultValue=0.3, parameterName=FaceDetailBrightness}, B=Constant() {r=0.02}, Alpha=NamedRerouteUsage())
   B=ScalarParameter() {defaultValue=0.85, parameterName=FaceBrightness}
   Alpha=@n10.B
  )
  n12 = ScalarParameter() {defaultValue=0.6, group=StuffRasp, parameterName=HeadAO}
  n13 = Add(
   A=Multiply(A=Power(Base=@n5.G), B=Constant4Vector() {constant=<rgba 2 2 2 1>}.RGB)
   B=Multiply(A=Multiply(A=@n9, B=@n11), B=LinearInterpolate(A=@n12, Alpha=@n10.B) {constA=0.7})
  )
  n14 = VectorParameter() {group=StuffRasp, parameterName=Emissive_Tint, defaultValue=<rgba 1 1 1 1>}
  MaterialFunctionCall(
   Material=SetMaterialAttributes(MaterialAttributes=@n4.MaterialAttributes, Emissive Color=Multiply(A=@n13, B=@n14.RGB)) {attributeSetTypes [#1]}
   M=@n1.M
   Pre Skinned Local Position=@n1.M
   Distance Blend=@n1.M
   Fresnel=@n1.M
  ) {materialFunction=&MF_CharacterParent_2023_PostFX}
  NamedRerouteDeclaration(None=@n3.FaceMask) {name=FaceMask, nodeColor=<rgba 1 0.6328 0 1>}
 ]
}
