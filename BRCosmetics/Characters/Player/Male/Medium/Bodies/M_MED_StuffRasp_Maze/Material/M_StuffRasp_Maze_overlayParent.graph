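# Translucent two-sided overlay parent material for the StuffRasp Maze character body, wrapped in CharacterParent PreFX and PostFX functions with sorted translucency.
# Opacity combines body masks, Fresnel, panning noise and sphere masks, and drops to zero on back faces.
# Base color blends a Fresnel gradient with desaturated glove shading, while emissive terms cover shoes and a switchable reactive glove glow.
# A procedural halftone dot mask uses rotated, tiled UVs.
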Material /BRCosmetics/Characters/Player/Male/Medium/Bodies/M_MED_StuffRasp_Maze/Material/M_StuffRasp_Maze_overlayParent
graph {
 properties {
  allowTranslucentCustomDepthWrites=true
  bAutomaticallySetUsageInEditor=false
  bCastRayTracedShadows=false
  bUseMaterialAttributes=true
  bUsedWithMeshDeformer=true
  bUsedWithSkeletalMesh=true
  blendMode=BLEND_Translucent
  ditherOpacityMask=true
  translucencyLightingMode=TLM_SurfacePerPixelLighting
  translucencyPass=MTP_BeforeDOF
  twoSided=true
  usedShadingModels=Default Lit
 }
 nodes [
  ComponentMask(None=NamedRerouteUsage()) {r=true}
  Constant()
  Constant() {r=1}
  Constant() {r=1}
  n1 = SetMaterialAttributes!(
   MaterialAttributes=NamedRerouteUsage()
   Emissive Color=NamedRerouteUsage()
   Base Color=NamedRerouteUsage()
   Metallic=ScalarParameter() {parameterName=StuffRasp_Metallic}
   Specular=NamedRerouteUsage()
   Roughness=NamedRerouteUsage()
   Anisotropy
   Opacity=MaterialFunctionCall!(Opacity=NamedRerouteUsage()) {materialFunction=/Game/Global/Materials/MaterialGym/Materials/SortedOutline/MF_SortTranslucency}.Result
   Opacity Mask
  ) {attributeSetTypes [
    #1
    #2
    #3
    #4
    #5
    #6
    #7
    #8
   ]}
  n2 = MaterialFunctionCall() {materialFunction=&MF_CharacterParent_2023_PreFX}
  MaterialFunctionCall(Material=@n1, M=@n2.M, Pre Skinned Local Position=@n2.M, Distance Blend=@n2.M, Fresnel=@n2.M) {materialFunction=&MF_CharacterParent_2023_PostFX}
  Multiply()
  n3 = NamedRerouteUsage()
  n4 = TextureSample() {
   samplerSource=SSM_Wrap_WorldGroupSettings
   samplerType=SAMPLERTYPE_Masks
   texture=/BRCosmetics/Characters/Player/Male/Medium/Bodies/M_MED_StuffRasp/Textures/T_StuffRasp_Body_M
  }
  n5 = Multiply(A=OneMinus(None=@n4.R), B=Constant() {r=0.5})
  n6 = Fresnel(ExponentIn=Constant() {r=0.07})
  n7 = Constant() {r=3.0965}
  n8 = LinearInterpolate(B=ScalarParameter() {defaultValue=0.9, parameterName=Global transparency}, Alpha=Multiply(A=Power(Base=@n6, Exp=@n7), B=Constant() {r=1})) {constB=0.9}
  n9 = TextureSample() {
   samplerSource=SSM_Wrap_WorldGroupSettings
   samplerType=SAMPLERTYPE_Masks
   texture=/BRCosmetics/Characters/Player/Male/Medium/Bodies/M_MED_StuffRasp/Textures/T_StuffRasp_Body_TAMask1
  }
  n10 = Multiply(A=OneMinus(None=@n9.R), B=@n9.G)
  n11 = Power(Base=Fresnel(ExponentIn=Constant() {r=1.635}, Normal=VertexNormalWS()))
  n12 = LinearInterpolate(
   A=Multiply(A=Add(A=@n5, B=@n8), B=@n10)
   B=LinearInterpolate(
    A=ScalarParameter() {defaultValue=0.35, group=StuffRasp, parameterName=StuffRasp_gloveOpacityMin}
    B=ScalarParameter() {defaultValue=0.9, group=StuffRasp, parameterName=StuffRasp_gloveOpacityMax}
    Alpha=@n11
   ) {constB=0}
   Alpha=ComponentMask(None=NamedRerouteUsage()) {r=true}
  ) {constB=0.5}
  n13 = TextureSample(UVs=Panner(Coordinate=NamedRerouteUsage()) {speedY=0.03}) {samplerSource=SSM_Wrap_WorldGroupSettings, samplerType=SAMPLERTYPE_Grayscale, texture=&LowResBlurredNoise}
  n14 = Multiply(A=@n13.R, B=Constant() {r=2})
  n15 = Constant() {r=2}
  n16 = Multiply(A=Saturate(None=Power(Base=@n14, Exp=@n15)), B=Constant() {r=0.05})
  n17 = SphereMask(
   A=VertexInterpolator(VS=LocalPosition() {localOrigin=InstancePreSkinning}.XYZ).PS
   B=Constant4Vector() {constant=<rgba 0 0 0 1>}.RGB
   Radius=Constant() {r=43.0183}
   Hardness=Constant()
  )
  n18 = TextureSample() {samplerType=SAMPLERTYPE_Masks, texture=&T_StuffRasp_Body_TAMask1}
  n19 = Add(A=@n12, B=Add(A=Multiply(A=@n12, B=@n16), B=Multiply(A=@n17, B=@n18.B)))
  n20 = TwoSidedSign()
  NamedRerouteDeclaration(None=LinearInterpolate(A=Constant(), B=Add(A=@n3, B=@n19), Alpha=Saturate(None=@n20)) {constA=1, constB=0}) {name=OutOpacity, nodeColor=<rgba 0 0.6719 1 1>}
  n21 = ScalarParameter() {defaultValue=0.01, group=StuffRasp, parameterName=StuffRasp_FresnelMin}
  n22 = ScalarParameter() {defaultValue=0.35, group=StuffRasp, parameterName=StuffRasp_FresnelMax}
  n23 = Power(
   Base=Fresnel(ExponentIn=ScalarParameter() {defaultValue=3, group=StuffRasp, parameterName=StuffRasp_FresnelExponentln})
   Exp=ScalarParameter() {defaultValue=1.5, parameterName=StuffRasp_FresnelPower}
  )
  n24 = Desaturation(None=NamedRerouteUsage(), Fraction=Constant() {r=1})
  n25 = ScalarParameter() {defaultValue=0.12, group=StuffRasp, parameterName=StuffRasp_gloveColor}
  n26 = NamedRerouteUsage()
  NamedRerouteDeclaration(None=LinearInterpolate(A=LinearInterpolate(A=@n21, B=@n22, Alpha=@n23), B=Multiply(A=@n24, B=@n25) {constB=0.05}, Alpha=ComponentMask(None=@n26) {r=true}) {constA=1, constB=0.05}) {name=OutBaseColor, nodeColor=<rgba 1 0 0.5859 1>}
  NamedRerouteDeclaration(None=MaterialFunctionCall!(
    TriplanarUV_UseClamp?=Constant()
    TriplanarUV_FlipUV=ScalarParameter() {group=TriplanarUV, parameterName=TriplanarUV_3D_FlipUV}
    TriplanarUV_CenterPosition=VectorParameter() {group=TriplanarUV, parameterName=TriplanarUV_3D_CenterPosition, defaultValue=<rgba 0 0 0 1>}.RGB
    TriplanarUV_UVScale=ScalarParameter() {defaultValue=82, group=TriplanarUV, parameterName=TriplanarUV_3D_UVSize}
   ) {materialFunction=/BRCosmetics/Characters/Player/Male/Medium/Bodies/M_MED_BirchPalace/Skins/Gas/Materials/MF_TriplanarUV}.UV_FrontProjection) {name=TriplannarUV_front, nodeColor=<rgba 1 0.3516 0 1>}
  NamedRerouteDeclaration(None=@n10) {name=HideMeshMask, nodeColor=<rgba 0.125 1 0 1>}
  NamedRerouteDeclaration(None=@n11) {name=GloveFresnel, nodeColor=<rgba 0.7812 1 0 1>}
  NamedRerouteDeclaration(None=SphereMask(
    A=NamedRerouteUsage()
    B=VectorParameter() {group=StuffRasp, parameterName=StuffRasp_LegCenterPosition, defaultValue=<rgba 0 0 0 1>}.RGB
    Radius=ScalarParameter() {defaultValue=56.6545, group=StuffRasp, parameterName=StuffRasp_LegMaskRadius}
    Hardness=ScalarParameter() {defaultValue=0.576, parameterName=StuffRasp_LegMaskHardness}
   )) {name=Leg gradient mask, nodeColor=<rgba 0 0.1094 1 1>}
  n27 = TextureSampleParameter2D() {
   desc="R = Spec , G = Roughness, B= Metal"
   group=00_Texture Inputs
   parameterName=SRM
   samplerSource=SSM_Wrap_WorldGroupSettings
   samplerType=SAMPLERTYPE_Masks
   sortPriority=1
   texture=&T_Fortnite_Default_S
  }
  NamedRerouteDeclaration(None=Multiply(A=@n27.R, B=ScalarParameter() {defaultValue=0.33, parameterName=StuffRasp_Specular})) {name=OutSpeclar, nodeColor=<rgba 0 0.9062 1 1>}
  NamedRerouteDeclaration(None=LinearInterpolate(
    A=ScalarParameter() {defaultValue=0.5, parameterName=StuffRasp_RoughnessMin}
    B=ScalarParameter() {defaultValue=1, parameterName=StuffRasp_RoughnessMax}
    Alpha=@n27.G
   ) {constA=0.5, constB=0.9}) {name=OutRoughness, nodeColor=<rgba 1 0.7734 0 1>}
  n28 = TextureSample() {
   samplerSource=SSM_Wrap_WorldGroupSettings
   samplerType=SAMPLERTYPE_Masks
   texture=/BRCosmetics/Characters/Player/Male/Medium/Bodies/M_MED_StuffRasp/Textures/T_StuffRasp_Body_TAMask3
  }
  n29 = VectorParameter() {group=StuffRasp, parameterName=ShoesEmissiveColor, defaultValue=<rgba 1 1 1 1>}
  n30 = ScalarParameter() {parameterName=StuffRasp_gloveColorSwitch}
  n31 = ScalarParameter() {parameterName=StuffRasp_gloveReactive}
  n32 = ScalarParameter() {defaultValue=-0.4063, parameterName=StuffRasp_gloveReactiveFresnnel}
  n33 = TextureSample() {
   samplerSource=SSM_Wrap_WorldGroupSettings
   samplerType=SAMPLERTYPE_Masks
   texture=&T_StuffRasp_Body_M
  }
  n34 = Multiply(A=@n33.G) {constB=1.05}
  n35 = Multiply(A=Multiply(A=Fresnel(ExponentIn=@n32), B=Constant() {r=20}) {constB=3}, B=Saturate(None=Power(Base=@n34) {constExponent=20}))
  n36 = Multiply(
   A=LinearInterpolate(B=VectorParameter() {parameterName=StuffRasp_gloveReactiveColor, defaultValue=<rgba 1.1 0.055 0.1652 1>}.RGB, Alpha=Multiply(A=@n30, B=@n31)) {constA=1}
   B=Multiply(A=@n35, B=@n31)
  )
  n37 = ComponentMask(None=NamedRerouteUsage()) {r=true}
  NamedRerouteDeclaration(None=Add(A=Multiply(A=@n28.R, B=@n29.RGB), B=LinearInterpolate(B=@n36, Alpha=@n37))) {name=ShoesEmissive, nodeColor=<rgba 0.4531 0 1 1>}
  n38 = GetMaterialAttributes(None=@n2.Result) {attributeGetTypes [#2]}
  NamedRerouteDeclaration(None=@n38.MaterialAttributes) {name=InMA, nodeColor=<rgba 1 0.2109 0 1>}
  NamedRerouteDeclaration(None=@n38.Base Color) {name=InBaseColor, nodeColor=<rgba 0 0.6719 1 1>}
  NamedRerouteDeclaration(None=@n2.Tech Art Mask) {name=FXMASK, nodeColor=<rgba 0.9922 1 0 1>}
  n39 = TextureSample(UVs=Multiply(A=TextureCoordinate()) {constB=2}) {samplerSource=SSM_Wrap_WorldGroupSettings, samplerType=SAMPLERTYPE_Grayscale, texture=&LowResBlurredNoise}
  n40 = Multiply(A=@n39.R, B=Constant() {r=0.7}) {constB=0.5}
  n41 = Saturate(None=Multiply(A=NamedRerouteUsage()) {constB=2})
  n42 = TextureSampleParameter2D() {
   group=00_Texture Inputs
   parameterName=Mask
   samplerSource=SSM_Wrap_WorldGroupSettings
   samplerType=SAMPLERTYPE_Masks
   sortPriority=3
   texture=&T_BlackMask
  }
  n43 = NamedRerouteDeclaration(None=Power(Base=@n42.G) {constExponent=3}) {name=GMap, nodeColor=<rgba 0 1 0.1797 1>}
  n44 = TextureCoordinate()
  n45 = ScalarParameter() {defaultValue=0.1, parameterName=Dot_Rotation}
  n46 = Frac(None=Multiply(
    A=MaterialFunctionCall(UVs=@n44, "Rotation Angle (0-1)"=@n45) {materialFunction=&CustomRotator}.Rotated Values
    B=ScalarParameter() {defaultValue=320, group=StuffRasp_axe, parameterName=Dot_Tiling}
   ))
  n47 = Constant() {r=0.5}
  n48 = LinearInterpolate(
   A=ScalarParameter() {group=StuffRasp_axe, parameterName=Dot_RadiusMax}
   B=ScalarParameter() {defaultValue=1.4, group=StuffRasp_axe, parameterName=Dot_RadiusMin}
   Alpha=Saturate(None=NamedRerouteUsage())
  )
  n49 = ScalarParameter() {defaultValue=0.9, parameterName=Dot_Hardness}
  n50 = NamedRerouteUsage()
  n51 = Multiply(
   A=Multiply(A=SphereMask(A=@n46, B=@n47, Radius=@n48, Hardness=@n49), B=LinearInterpolate(Alpha=@n50) {constA=1, constB=0})
   B=NamedRerouteUsage()
  )
  NamedRerouteDeclaration(None=Multiply(A=LinearInterpolate(A=@n40, Alpha=@n41) {constA=1}, B=Multiply(A=@n43, B=@n51))) {name=Dot_Mask, nodeColor=<rgba 0 0.4609 1 1>}
  NamedRerouteDeclaration(None=@n2.Pre skinned Local Position) {name=PreskinPosition, nodeColor=<rgba 0.7578 1 0 1>}
  Saturate(None=@n35)
 ]
}
